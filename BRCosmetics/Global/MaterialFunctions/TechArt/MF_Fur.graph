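# Fur material function that fakes fur on surfaces with parallax occlusion mapping of a tiled fur heightmap, returning material attributes.
# Parallax step counts use temporal dither and a depth-based fade; Low and Medium quality use plain tiled UVs.
# Fur noise, tint and Fresnel shape base color, blended toward GrimeColor by Grime.
# Fur alpha drives modulated roughness; a flattened, larger-scale waterfall normal adds waviness, and the normalflip multiply is unconnected.
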MaterialFunction /BRCosmetics/Global/MaterialFunctions/TechArt/MF_Fur
graph {nodes [
  n1 = ScalarParameter() {defaultValue=0.5, parameterName=Fmin}
  n2 = ScalarParameter() {defaultValue=3, parameterName=Fmax}
  n3 = Fresnel()
  n4 = FunctionInput(Preview=VectorParameter() {parameterName=FurColor, defaultValue=<rgba 0.55 0.4955 0.418 1>}.RGB) {bUsePreviewValueAsDefault=true, inputName=FurColor}
  n5 = ScalarParameter() {defaultValue=2, parameterName=Min Steps}
  n6 = LinearInterpolate(
   B=ScalarParameter() {defaultValue=1, parameterName=TAA StepMult}
   Alpha=MaterialFunctionCall() {materialFunction=&DitherTemporalAA}.Result
  ) {constA=1, constB=1.2}
  n7 = Clamp(None=SphereMask(A=PixelDepth().R, B=Constant(), Radius=ScalarParameter() {defaultValue=1024, desc=756, parameterName=FadeRadius}) {attenuationRadius=5000, hardnessPercent=0})
  n8 = ScalarParameter() {defaultValue=4, parameterName=Max Steps}
  n9 = FunctionInput(Preview=ScalarParameter() {defaultValue=15, parameterName=Tiling}) {bUsePreviewValueAsDefault=true, inputName=Tiling, inputType=FunctionInput_Scalar, previewValue=<1 0 0 1>}
  n10 = Multiply(A=TextureCoordinate(), B=@n9)
  n11 = MaterialFunctionCall!(
   Heightmap Texture=TextureObjectParameter() {parameterName=Height, texture=/BRCosmetics/Global/Textures/Default/Noise/FUR_M}
   Height Ratio=ScalarParameter() {defaultValue=0.8, parameterName=FurHeight}
   Min Steps=Multiply(A=Multiply(A=@n5, B=@n6), B=@n7)
   Max Steps=Multiply(A=Multiply(A=@n8, B=@n6), B=@n7)
   UVs=@n10
   Heightmap Channel=Constant4Vector() {constant=<rgba 0 0 0 1>}.RGBA
   Reference Plane
   -----------------------------
   Use World Coordinates
   Specify Manual Texture Size
   Manual Texture Size
   ------------------------------
   "Render Shadows (Occlusion Mapping)"=StaticBoolParameter() {defaultValue=true, parameterName=Shadows}
   Light Vector=Normalize(VectorInput=VectorParameter() {parameterName=LV, defaultValue=<rgba 0 -4.0151 -1 0>}.RGB)
   Shadow Steps=ScalarParameter() {defaultValue=8, parameterName=st}
   Shadow Penumbra=ScalarParameter() {defaultValue=1, parameterName=k}
   Transform To VertexNormal
  ) {materialFunction=/Engine/Functions/Engine_MaterialFunctions01/Texturing/ParallaxOcclusionMapping}
  n12 = QualitySwitch(Default=@n11.Parallax UVs, Low=@n10, High=@n11.Parallax UVs, Medium=@n10)
  n13 = TextureCoordinate()
  n14 = DDX(Value=@n13)
  n15 = DDY(Value=@n13)
  n16 = TextureSampleParameter2D!(UVs=@n12, "DDX(UVs)"=@n14, "DDY(UVs)"=@n15, Apply View MipBias) {mipValueMode=TMVM_Derivative, parameterName=Base, texture=&FUR_M}
  n17 = Power(
   Base=@n16.G
   Exp=LinearInterpolate(
    A=ScalarParameter() {defaultValue=0.33, parameterName=FurNoiseClose}
    B=ScalarParameter() {defaultValue=0.8, parameterName=FurNoiseFar}
    Alpha=@n7
   ) {constA=1, constB=0.5}
  )
  n18 = Multiply(
   A=ScalarParameter() {defaultValue=1, parameterName=Base Color}
   B=Multiply(A=LinearInterpolate(A=@n1, B=@n2, Alpha=@n3), B=Multiply(A=@n4, B=@n17))
  )
  n19 = FunctionInput(Preview=Constant3Vector() {constant=<rgba 0.125 0.1186 0.1065 1>}) {bUsePreviewValueAsDefault=true, inputName=GrimeColor}
  n20 = FunctionInput() {bUsePreviewValueAsDefault=true, inputName=Grime}
  n21 = LinearInterpolate(A=ScalarParameter() {defaultValue=0.8, parameterName=O1}, B=ScalarParameter() {defaultValue=1, parameterName=O2}, Alpha=@n16.A)
  n22 = FunctionInput() {bUsePreviewValueAsDefault=true, inputName=RoughnessModulation, inputType=FunctionInput_Scalar, previewValue=<1 0 0 1>}
  n23 = TextureSampleParameter2D(UVs=Multiply(A=@n12) {constB=0.33}, "DDX(UVs)"=@n14, "DDY(UVs)"=@n15) {
   mipValueMode=TMVM_Derivative
   parameterName=Normal
   samplerType=SAMPLERTYPE_Normal
   texture=/Game/Tools/WaterfallGenerator/Art/T_Waterfall_NormalA
  }
  n24 = ScalarParameter() {defaultValue=0.8, parameterName=FurWavyness}
  FunctionOutput(None=MakeMaterialAttributes!(
    BaseColor=LinearInterpolate(A=@n18, B=@n19, Alpha=@n20)
    Metallic
    Specular
    Roughness=Multiply(A=@n21, B=@n22)
    Anisotropy
    EmissiveColor
    Opacity
    OpacityMask
    Normal=MaterialFunctionCall(Normal=@n23.RGB, Flatness=@n24) {materialFunction=&FlattenNormal}.Result
    Tangent
    WorldPositionOffset
    SubsurfaceColor
    ClearCoat
    ClearCoatRoughness
    AmbientOcclusion
    Refraction
    CustomizedUVs_0
    CustomizedUVs_1
    CustomizedUVs_2
    CustomizedUVs_3
    CustomizedUVs_4
    CustomizedUVs_5
    CustomizedUVs_6
    CustomizedUVs_7
    PixelDepthOffset
    ShadingModel
    Displacement
   ))
  Multiply(A=@n23.RGB, B=VectorParameter() {parameterName=normalflip, defaultValue=<rgba 1 1 1 1>}.RGB)
 ]}
