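# Material function adding reactive emissive to the ZirconSweep character: a timed progress glow plus elimination pulses.
# Pre-skinned position builds body and arm gradients, blended by mask, along which progress eases from previous to current value.
# Pattern and panning noise shape the glow; pulse logic is duplicated for rapid successive eliminations.
# A master amount scales emissive, optional base color darkening and the output reactivity mask, allowing smooth Frontend looping.
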MaterialFunction /Game/Characters/Player/Male/Medium/Bodies/M_MED_ZirconSweep/Materials/MF_M_MED_ZirconSweep_Reactivity
graph {
 properties {description=Reacitivity Control for ZirconSweep character}
 nodes [
  n1 = FunctionInput() {inputName=InMaterial, inputType=FunctionInput_MaterialAttributes}
  n2 = GetMaterialAttributes(None=@n1) {attributeGetTypes [#1, #2]}
  n3 = Time()
  n4 = ScalarParameter() {defaultValue=4, group=Reactivity, parameterName=ElimPulseDuration}
  n5 = ScalarParameter() {defaultValue=-100, group=Reactivity, parameterName=ElimPulseStartTime_CCPM, sliderMax=1}
  n6 = Divide(A=StaticSwitchParameter(True=Fmod(A=@n3, B=@n4), False=Subtract(A=@n3, B=@n5)) {group=Reactivity, parameterName=ElimPulse_DEBUG}, B=@n4)
  n7 = OneMinus(None=Saturate(None=@n6))
  n8 = ScalarParameter() {defaultValue=0.65, group=Reactivity, parameterName=ElimPulseProgressExp}
  n9 = ScalarParameter() {defaultValue=0.25, group=Reactivity, parameterName=ElimPulseWidth, sliderMax=0.5}
  n10 = NamedRerouteUsage()
  n11 = OneMinus(None=Power(Base=OneMinus(None=@n10), Exp=@n8) {constExponent=0.4})
  n12 = Constant()
  n13 = ScalarParameter() {defaultValue=0.05, group=Reactivity, parameterName=ReactiveEmissivePanSpeed}
  n14 = TextureSampleParameter2D() {
   group=Reactivity
   parameterName=ReactiveMasks
   samplerSource=SSM_Wrap_WorldGroupSettings
   samplerType=SAMPLERTYPE_Masks
   texture=/Game/Characters/Player/Male/Medium/Bodies/M_MED_ZirconSweep/Textures/T_M_MED_ZirconSweep_Body_TechArtFX_Masks
   channelNames {a="", b=Mirror Mask, g=Gradient Mask, r=""}
  }
  n15 = Panner(Coordinate=TextureCoordinate(), Speed=Multiply(A=AppendVector(A=@n12, B=@n13), B=LinearInterpolate(Alpha=@n14.Mirror Mask) {constA=-1})) {speedY=0.3}
  n16 = TextureSampleParameter2D(UVs=Multiply(A=@n15) {constB=2}) {
   group=Reactivity
   parameterName=ReactiveNoise
   samplerSource=SSM_Wrap_WorldGroupSettings
   samplerType=SAMPLERTYPE_LinearColor
   texture=/Game/Global/Textures/Default/Tileable_Patterns/Cloud_Diffuse
  }
  n17 = Multiply(
   A=ConstantBiasScale(None=@n16.R) {bias=-0.5, scale=2}
   B=ScalarParameter() {defaultValue=0.06, group=Reactivity, parameterName=ElimPulseDistortionAmt, sliderMax=0.1}
  )
  n18 = LinearInterpolate(A=@n9, Alpha=Add(A=@n11, B=Multiply(A=@n11, B=@n17))) {constA=0.15}
  n19 = Power(Base=SphereMask(A=Power(Base=@n6, Exp=@n8) {constExponent=0.78}, B=@n18, Radius=@n9) {attenuationRadius=0.1, hardnessPercent=0}) {constExponent=10}
  n20 = Divide(A=Subtract(A=@n3, B=ScalarParameter() {defaultValue=-100, group=Reactivity, parameterName=ElimPulseStartTime2_CCPM, sliderMax=1}), B=@n4)
  n21 = OneMinus(None=Saturate(None=@n20))
  n22 = Power(Base=SphereMask(A=Power(Base=@n20, Exp=@n8) {constExponent=0.78}, B=@n18, Radius=@n9) {attenuationRadius=0.1, hardnessPercent=0}) {constExponent=10}
  n23 = TextureSampleParameter2D() {
   group=Reactivity
   parameterName=ReactivePattern
   samplerSource=SSM_Wrap_WorldGroupSettings
   samplerType=SAMPLERTYPE_Alpha
   texture=/Game/Characters/Player/Male/Medium/Bodies/M_MED_ZirconSweep/Textures/T_M_MED_ZirconSweep_Body_TechArtFX_Pattern
   channelNames {a="", b="", g="", r=Pattern Mask}
  }
  n24 = Multiply(
   A=Add(A=Multiply(A=@n7, B=@n19), B=Multiply(A=@n21, B=@n22)) {desc=Pulse logic repeated two times to allow for rapid subsequent eliminations}
   B=@n23.Pattern Mask
  )
  n25 = VectorParameter() {group=Reactivity, parameterName=ElimPulseEmissiveColor, defaultValue=<rgba 0.0208 0.6016 1 1>}
  n26 = VectorParameter() {group=Reactivity, parameterName=ElimPulseEmissiveColor2, defaultValue=<rgba 0.0208 0.6016 1 1>}
  n27 = NamedRerouteUsage()
  n28 = Multiply(
   A=Multiply(A=@n24, B=ScalarParameter() {defaultValue=40, group=Reactivity, parameterName=ElimPulseIntensity}) {constB=60}
   B=StaticSwitchParameter(True=LinearInterpolate(A=@n25.RGB, B=@n26.RGB, Alpha=@n27), False=@n25.RGB) {group=Reactivity, parameterName=ElimPulseColorGradientEnabled}
  )
  n29 = ScalarParameter() {defaultValue=5, group=Reactivity, parameterName=Reactive_DEBUG_NumSteps}
  n30 = Floor(None=Fmod(A=Time(), B=Add(A=@n29) {constB=2}))
  n31 = NamedRerouteDeclaration(None=Subtract(A=Time(), B=ScalarParameter() {defaultValue=-100, group=Reactivity, parameterName=ReactiveStartTime_CCPM, sliderMax=1})) {name=ReactivityRelativeTime, nodeColor=<rgba 0 0.0156 1 1>}
  n32 = ScalarParameter() {defaultValue=3.5, group=Reactivity, parameterName=ReactiveDuration_CCPM}
  n33 = LinearInterpolate(
   A=ScalarParameter() {group=Reactivity, parameterName=ReactivePreviousProgress_CCPM, sliderMax=1}
   B=StaticSwitchParameter(True=Divide(A=@n30) {constB=5}, False=ScalarParameter() {group=Reactivity, parameterName=ReactiveProgress_CCPM, sliderMax=1}) {group=Reactivity, parameterName=Reactive_DEBUG}
   Alpha=Saturate(None=Divide(A=@n31, B=@n32))
  )
  n34 = InverseLinearInterpolate(
   A=LinearInterpolate(Alpha=@n33) {constA=-0.05, constB=1.1, desc="Gradient stop edge (full black)"}
   B=LinearInterpolate(Alpha=@n33) {constA=-0.15, desc="Gradient start edge (full white)"}
   Value=NamedRerouteUsage()
  )
  n35 = NamedRerouteDeclaration(None=Power(Base=Saturate(None=@n34), Exp=ScalarParameter() {defaultValue=2, group=Reactivity, parameterName=ReactiveEmissiveGradientExp}) {constExponent=10}) {name=CurrentProgressGradient, nodeColor=<rgba 1 0 0.5859 1>}
  n36 = TextureSampleParameter2D() {
   group=Reactivity
   parameterName=ReactivePattern
   samplerSource=SSM_Wrap_WorldGroupSettings
   samplerType=SAMPLERTYPE_Alpha
   texture=&T_M_MED_ZirconSweep_Body_TechArtFX_Pattern
   channelNames {a="", b="", g="", r=Pattern Mask}
  }
  n37 = Multiply(A=@n35, B=@n36.Pattern Mask) {constA=1}
  n38 = TextureSampleParameter2D(UVs=@n15) {
   group=Reactivity
   parameterName=ReactiveNoise
   samplerSource=SSM_Wrap_WorldGroupSettings
   samplerType=SAMPLERTYPE_LinearColor
   texture=&Cloud_Diffuse
  }
  n39 = LinearInterpolate(
   A=@n37
   B=Multiply(A=@n37, B=Power(Base=@n38.R) {constExponent=3})
   Alpha=ScalarParameter() {defaultValue=0.8, group=Reactivity, parameterName=ReactiveNoiseIntensity, sliderMax=1}
  ) {constAlpha=0.9}
  n40 = VectorParameter() {group=Reactivity, parameterName=ReactiveEmissiveColor, defaultValue=<rgba 1 1 0 1>}
  n41 = InverseLinearInterpolate(B=@n33, Value=NamedRerouteUsage())
  n42 = LinearInterpolate(
   A=@n40.RGB
   B=VectorParameter() {group=Reactivity, parameterName=ReactiveEmissiveColor2, defaultValue=<rgba 1 0 0 1>}.RGB
   Alpha=Power(Base=Saturate(None=@n41), Exp=ScalarParameter() {defaultValue=0.8, group=Reactivity, parameterName=ReactiveColorGradientExp})
  )
  n43 = Multiply(
   A=@n39
   B=Multiply(
    A=StaticSwitchParameter(True=@n42, False=@n40.RGB) {group=Reactivity, parameterName=ReactiveColorGradientEnabled}
    B=ScalarParameter() {defaultValue=7, group=Reactivity, parameterName=ReactiveEmissiveIntensity}
   )
  )
  n44 = ScalarParameter() {defaultValue=1, desc=Used in Frontend for smooth looping of reactivity, parameterName=ReactiveMasterAmount_CCPM, sliderMax=1}
  n45 = Add(A=@n2.EmissiveColor, B=Multiply(A=Add(A=@n28, B=@n43), B=@n44))
  n46 = ScalarParameter() {defaultValue=0.5, group=Reactivity, parameterName=ReactiveBaseColorBrightnessScale, sliderMax=1}
  n47 = Multiply(A=@n36.Pattern Mask) {constB=4}
  n48 = TextureSampleParameter2D() {
   group=Reactivity
   parameterName=ReactiveMasks
   samplerSource=SSM_Wrap_WorldGroupSettings
   samplerType=SAMPLERTYPE_Masks
   texture=&T_M_MED_ZirconSweep_Body_TechArtFX_Masks
   channelNames {a="", b=Mirror Mask, g=Gradient Mask, r=Desaturation Mask}
  }
  n49 = Multiply(A=Multiply(A=@n35, B=Saturate(None=@n47)), B=@n48.Desaturation Mask)
  n50 = StaticSwitchParameter(True=LinearInterpolate(A=@n2.BaseColor, B=Multiply(A=@n2.BaseColor, B=@n46), Alpha=Min(A=@n44, B=@n49)) {constB=0}, False=@n2.BaseColor) {group=Reactivity, parameterName=ReactiveBaseColorReduction}
  FunctionOutput(None=StaticSwitchParameter(
    True=SetMaterialAttributes!(MaterialAttributes=@n2.MaterialAttributes, EmissiveColor=@n45, BaseColor=@n50) {attributeSetTypes [#1, #2]}
    False=@n1
   ) {defaultValue=true, group=Reactivity, parameterName=ReactiveEmissiveEnabled, sortPriority=0})
  FunctionOutput(None=NamedRerouteUsage()) {outputName=ReactivityMask, sortPriority=1}
  n51 = MaterialFunctionCall(Float3=VertexInterpolator(VS=PreSkinnedPosition()).PS) {materialFunction=&BreakOutFloat3Components}
  n52 = MaterialFunctionCall(X=Abs(None=@n51.R), Y=@n51.G, Z=@n51.G) {materialFunction=&MakeFloat3}
  n53 = Constant3Vector() {desc=Crown, constant=<rgba 0 8 140 0>}
  n54 = Constant3Vector() {desc="Body Gradient  axis", constant=<rgba 0 0 -1 1>}
  n55 = Divide(
   A=DotProduct(A=Subtract(A=@n52.Result, B=@n53), B=Normalize(VectorInput=@n54))
   B=Distance(A=@n53, B=Constant3Vector() {desc=Coat Base, constant=<rgba 0 8 55 0>})
  )
  n56 = ConstantBiasScale(None=Saturate(None=Abs(None=@n55)))
  n57 = Constant3Vector() {desc=Hand, constant=<rgba 60 8 88 0>}
  n58 = DotProduct(A=Subtract(A=@n52.Result, B=@n57), B=Normalize(VectorInput=Constant3Vector() {desc=Arm Gradient axis, constant=<rgba -1 -0.25 0.75 1>}))
  n59 = Distance(A=@n57, B=Constant3Vector() {desc=Crown, constant=<rgba 0 8 150 0>})
  n60 = Multiply(A=Saturate(None=Divide(A=@n58, B=@n59))) {constB=0.7}
  n61 = TextureSampleParameter2D() {
   group=Reactivity
   parameterName=ReactiveMasks
   samplerSource=SSM_Wrap_WorldGroupSettings
   samplerType=SAMPLERTYPE_Masks
   texture=&T_M_MED_ZirconSweep_Body_TechArtFX_Masks
   channelNames {a="", b=Mirror Mask, g=Gradient Mask, r=""}
  }
  n62 = NamedRerouteDeclaration(None=Saturate(None=LinearInterpolate(A=@n56, B=@n60, Alpha=@n61.Gradient Mask))) {name=ReactivityProgressGradient, nodeColor=<rgba 1 0.1875 0 1>}
  Multiply(
   A=MaterialFunctionCall(Hue Shift Percentage=@n62, Texture=Constant3Vector() {constant=<rgba 1 0 0 0>}) {materialFunction=&HueShift}.Result
   B=Frac(None=Multiply(A=@n62) {constB=7})
  )
  NamedRerouteDeclaration(None=Multiply(A=@n44, B=Add(A=@n24, B=@n39))) {name=ReactivityMask_OUT, nodeColor=<rgba 0 0.0625 1 1>}
 ]
}
