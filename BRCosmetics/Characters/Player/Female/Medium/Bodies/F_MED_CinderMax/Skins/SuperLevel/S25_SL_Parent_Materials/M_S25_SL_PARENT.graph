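# Parent character material for Season 25 SuperLevel skins, layering recolorable, animated styles between the shared CharacterParent_2023 PreFX and PostFX functions.
# SL_FX_Mask channels split the surface into primary, secondary, tertiary and emissive regions, each with fresnel-driven shading, while unmasked areas get separate tinting.
# Primary patterns use animated triplanar or cylindrical projection.
# Static switches select T1 fur, T2 rainbow/panning, T3 patterns, hair, ZirconSweep reactivity and debug views.
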Material /BRCosmetics/Characters/Player/Female/Medium/Bodies/F_MED_CinderMax/Skins/SuperLevel/S25_SL_Parent_Materials/M_S25_SL_PARENT
graph {
 properties {
  bUseMaterialAttributes=true
  bUsedWithMeshDeformer=true
  bUsedWithMorphTargets=true
  bUsedWithSkeletalMesh=true
  pixelDepthOffsetMode=PDOM_Legacy
  shadingModel=MSM_FromMaterialExpression
  usedShadingModels=Default Lit | Subsurface | Clear Coat | Subsurface Profile | Cloth
 }
 nodes [
  n1 = MaterialFunctionCall() {materialFunction=&MF_CharacterParent_2023_PreFX}
  MaterialFunctionCall(Material=NamedRerouteUsage(), M=@n1.M, Pre Skinned Local Position=@n1.M, Distance Blend=@n1.M, Fresnel=@n1.M) {materialFunction=&MF_CharacterParent_2023_PostFX}
  NamedRerouteDeclaration(None=@n1.Result) {name=MA_IN, nodeColor=<rgba 0.7344 0 1 1>}
  n2 = NamedRerouteUsage()
  n3 = StaticSwitchParameter(True=StaticSwitchParameter(True=@n2, False=Constant()) {group=*S25_SL_Hair, parameterName=UseHairEmissive}, False=@n2) {group=*S25_SL, parameterName=UseHair}
  n4 = NamedRerouteUsage()
  n5 = NamedRerouteUsage()
  n6 = SetMaterialAttributes!(MaterialAttributes, Base Color=Multiply(A=@n4, B=@n5)) {attributeSetTypes [#1]}
  n7 = Constant()
  n8 = SetMaterialAttributes!(MaterialAttributes, Emissive Color=NamedRerouteUsage(), Base Color=@n7, Metallic=@n7, Roughness=Constant() {r=1}) {attributeSetTypes [
    #2
    #1
    #3
    #4
   ]}
  n9 = NamedRerouteUsage()
  n10 = SetMaterialAttributes!(MaterialAttributes=@n9, Base Color=NamedRerouteUsage(), Metallic=NamedRerouteUsage(), Emissive Color=NamedRerouteUsage()) {attributeSetTypes [#1, #3, #2]}
  n11 = Saturate(None=SphereMask(A=NamedRerouteUsage(), B=Constant3Vector() {constant=<rgba 0 10 130 1>}) {attenuationRadius=20})
  n12 = GetMaterialAttributes(None=NamedRerouteUsage()) {attributeGetTypes [
    #1
    #4
    #3
    #5
    #2
    #6
   ]}
  n13 = SetMaterialAttributes!(
   MaterialAttributes=@n12.MaterialAttributes
   Base Color=NamedRerouteUsage()
   Roughness=NamedRerouteUsage()
   Metallic=NamedRerouteUsage()
   Specular=NamedRerouteUsage()
   Emissive Color=NamedRerouteUsage()
   Normal=NamedRerouteUsage()
  ) {attributeSetTypes [
    #1
    #4
    #3
    #5
    #2
    #6
   ]}
  n14 = Multiply(A=@n5, B=NamedRerouteUsage())
  n15 = StaticSwitchParameter(True=Saturate(None=Subtract(A=@n4, B=@n14)), False=@n4) {group=*S25_SL, parameterName=UseZirconSweepReactivity}
  n16 = Saturate(None=Power(Base=@n15, Exp=ScalarParameter() {defaultValue=1, group=*S25_SL, parameterName=SkinPower}))
  n17 = Power(Base=@n15, Exp=ScalarParameter() {defaultValue=3, group=*S25_SL, parameterName=SkinMask_Power})
  n18 = NamedRerouteUsage()
  n19 = StaticSwitchParameter(
   True=StaticSwitchParameter(True=BlendMaterialAttributes(A=@n10, B=@n9, Alpha=@n11) {desc=blend skin}, False=@n10) {group=*S25_SL_Hair, parameterName=UseZirconFaceAccHideBeardSL}
   False=BlendMaterialAttributes(A=@n13, B=NamedRerouteUsage(), Alpha=LinearInterpolate(A=@n16, B=@n17, Alpha=@n18)) {desc=blend skin}
  ) {group=*S25_SL, parameterName=UseHair}
  n20 = MaterialFunctionCall!(InMaterial=@n19) {materialFunction=/Game/Characters/Player/Male/Medium/Bodies/M_MED_ZirconSweep/Materials/MF_M_MED_ZirconSweep_Reactivity}
  n21 = GetMaterialAttributes(None=@n20.Result) {attributeGetTypes [#2]}
  n22 = GetMaterialAttributes(None=@n19) {attributeGetTypes [
    #2
    #1
    #3
    #5
   ]}
  n23 = NamedRerouteUsage()
  n24 = Constant() {r=-2}
  n25 = Multiply(
   A=Multiply(A=Desaturation(None=@n23, Fraction=@n24), B=@n20.ReactivityMask)
   B=ScalarParameter() {defaultValue=20, group=*S25_SL, parameterName=RainbowReactivity_EmisiiveIntensity}
  ) {constB=10}
  n26 = Saturate(None=Multiply(A=@n20.ReactivityMask) {constB=20})
  n27 = SetMaterialAttributes!(
   MaterialAttributes=@n21.MaterialAttributes
   Emissive Color=LinearInterpolate(A=@n22.Emissive Color, B=StaticSwitchParameter(True=@n25, False=@n21.Emissive Color) {group=*S25_SL, parameterName=UseT2}, Alpha=@n26)
   Base Color=LinearInterpolate(A=@n22.Base Color, Alpha=@n26) {constB=0}
   Metallic=LinearInterpolate(A=@n22.Metallic, Alpha=@n26) {constB=0}
   Specular=LinearInterpolate(A=@n22.Specular, Alpha=@n26) {constB=0}
   Opacity=Constant()
  ) {attributeSetTypes [
    #2
    #1
    #3
    #5
    #7
   ]}
  n28 = StaticSwitchParameter(True=@n8, False=StaticSwitchParameter(True=@n27, False=@n19) {group=*S25_SL, parameterName=UseZirconSweepReactivity}) {group=*S25_SL, parameterName=UseDebug_SL_Mask}
  NamedRerouteDeclaration(None=StaticSwitchParameter(
    True=SetMaterialAttributes!(MaterialAttributes, Base Color=@n3) {attributeSetTypes [#1]}
    False=StaticSwitchParameter(True=@n6, False=@n28) {group=*S25_SL, parameterName=UseDebug_Color}
   ) {group=*S25_SL, parameterName=UseDebug_Emissive}) {name=MA_OUT, nodeColor=<rgba 0 1 0.2031 1>}
  NamedRerouteDeclaration(None=VectorParameter() {group=*S25_SL, parameterName=SecondaryColor, sortPriority=1, defaultValue=<rgba 1 0 0 1>}.RGB) {name=SecondaryColor, nodeColor=<rgba 0.9507 0.099 1 1>}
  NamedRerouteDeclaration(None=VectorParameter() {group=*S25_SL, parameterName=TertiaryColor, sortPriority=2, defaultValue=<rgba 1 0 0 1>}.RGB) {name=TertiaryColor, nodeColor=<rgba 0.9721 0.4896 1 1>}
  n29 = NamedRerouteUsage()
  n30 = StaticSwitchParameter(True=Constant(), False=NamedRerouteUsage()) {group=*S25_SL, parameterName=UseT3}
  n31 = Multiply(
   A=ScalarParameter() {defaultValue=0.9, group=*S25_SL, parameterName=PrimaryPattern_Strength}
   B=SmoothStep(
    Min=ScalarParameter() {defaultValue=0.73, group=*S25_SL, parameterName=PrimaryPattern_Range_Min}
    Max=ScalarParameter() {defaultValue=0.76, group=*S25_SL, parameterName=PrimaryPattern_Range_Max}
    Value=Subtract(A=@n29, B=@n30)
   ) {constMax=0.895, constMin=0.89}
  )
  n32 = Add(
   A=Divide(A=NamedRerouteUsage(), B=Constant() {r=128})
   B=Multiply(A=Time(), B=ScalarParameter() {defaultValue=0.05, group=*S25_SL, parameterName=VerticalPanningSpeed})
  )
  n33 = Distance(A=Sine(None=ComponentMask(None=@n32) {b=true}), B=Constant())
  n34 = NamedRerouteUsage()
  n35 = Saturate(None=Add(A=SmoothStep(Value=@n34)) {constB=0.3})
  n36 = Multiply(A=LinearInterpolate(A=Power(Base=@n33) {constExponent=3}, B=NamedRerouteUsage()), B=@n35)
  n37 = Multiply(A=@n31, B=Saturate(None=LinearInterpolate(A=@n36, B=@n35, Alpha=@n35)))
  NamedRerouteDeclaration(None=OneMinus(None=StaticSwitchParameter(True=@n37, False=@n31) {group=*S25_SL, parameterName=UseT2})) {name=PrimaryPattern_Mask, nodeColor=<rgba 0.6875 0 1 1>}
  MaterialExpressionLuminance_0 = MaterialXLuminance!(None=NamedRerouteUsage())
  n38 = NamedRerouteUsage()
  n39 = Multiply(A=NamedRerouteUsage(), B=Saturate(None=Add(A=@n38) {constB=0.1}))
  n40 = ScalarParameter() {defaultValue=1, group=*S25_SL, parameterName=NotSL_TintIntensity}
  NamedRerouteDeclaration(None=LinearInterpolate(A=NamedRerouteUsage(), B=LinearInterpolate(A=@MaterialExpressionLuminance_0, B=@n39, Alpha=@n40) {constAlpha=1}, Alpha=NamedRerouteUsage())) {name=BaseColor_OUT, nodeColor=<rgba 0.3505 1 0 1>}
  n41 = LinearInterpolate(
   A=LinearInterpolate(A=NamedRerouteUsage(), B=NamedRerouteUsage(), Alpha=NamedRerouteUsage())
   B=NamedRerouteUsage()
   Alpha=NamedRerouteUsage()
  )
  n42 = NamedRerouteUsage()
  n43 = NamedRerouteUsage()
  n44 = NamedRerouteUsage()
  n45 = ScalarParameter() {defaultValue=0.3, group=*S25_SL, parameterName=Quaternary_BaseColor_Multi}
  NamedRerouteDeclaration(None=LinearInterpolate(A=LinearInterpolate(A=@n41, B=@n42, Alpha=@n43), B=Multiply(A=@n44, B=@n45), Alpha=NamedRerouteUsage()) {constB=0}) {name=SL_BaseColor, nodeColor=<rgba 0 1 0.4609 1>}
  n46 = NamedRerouteUsage()
  n47 = CurveAtlasRowParameter(CurveTime=NamedRerouteUsage()) {
   atlas=/BRCosmetics/Characters/Player/Female/Medium/Bodies/F_MED_CinderMax/Skins/SuperLevel/S25_SL_Parent_Materials/Textures/CA_S25_SL_Primary
   curve=/BRCosmetics/Characters/Player/Female/Medium/Bodies/F_MED_CinderMax/Skins/SuperLevel/S25_SL_Parent_Materials/Textures/C_S25_SL_Primary_T1
   desc=background noise
   group=*S25_SL
   parameterName=Primary_Base_Curve
  }
  n48 = ScalarParameter() {group=*S25_SL, parameterName=Primary_BackgroundNoiseMin}
  n49 = NamedRerouteUsage()
  n50 = NamedRerouteUsage()
  n51 = ScalarParameter() {defaultValue=2.8, group=*S25_SL, parameterName=Primary_FresnelPower_Outter}
  n52 = OneMinus(None=NamedRerouteUsage())
  n53 = ScalarParameter() {defaultValue=15, group=*S25_SL, parameterName=Primary_FresnelPower_Inner}
  n54 = Add(
   A=Multiply(A=Power(Base=@n50, Exp=@n51), B=ScalarParameter() {defaultValue=3, group=*S25_SL, parameterName=Primary_FresnelIntensity_Outter})
   B=Multiply(
    A=Power(Base=@n52, Exp=@n53) {constExponent=10}
    B=ScalarParameter() {defaultValue=0.1, group=*S25_SL, parameterName=Primary_FresnelIntensity_Inner}
   )
  )
  n55 = Multiply(A=@n49, B=NamedRerouteDeclaration(None=Saturate(None=@n54)) {name=Primary_Fresnel, nodeColor=<rgba 1 0.3473 0.2776 1>})
  n56 = LinearInterpolate(Alpha=@n55) {constA=0.25, desc=Leopard}
  n57 = LinearInterpolate(
   A=Multiply(A=NamedRerouteUsage(), B=NamedRerouteUsage())
   B=Multiply(A=StaticSwitchParameter(True=@n46, False=@n47) {group=*S25_SL, parameterName=UseT2}, B=SmoothStep(Min=@n48, Value=@n56) {constMin=0.5})
   Alpha=@n49
  )
  n58 = GetMaterialAttributes(None=MaterialFunctionCall!(Tiling, Grime, GrimeColor, RoughnessModulation, FurColor=@n57) {materialFunction=/BRCosmetics/Global/MaterialFunctions/TechArt/MF_Fur}.Result) {attributeGetTypes [#1, #4, #6]}
  n59 = NamedRerouteDeclaration(None=@n58.Base Color) {name=Fur_BaseColor, nodeColor=<rgba 0.9453 1 0 1>}
  NamedRerouteDeclaration(None=Max(A=StaticSwitchParameter(True=@n59, False=@n57) {parameterName=UseT1}) {constB=0.009}) {name=Primary_BaseColor, nodeColor=<rgba 1 0 0.1875 1>}
  n60 = OneMinus(None=NamedRerouteUsage())
  n61 = NamedRerouteUsage()
  n62 = ScalarParameter() {defaultValue=1, group=*S25_SL, parameterName=Perlin_Intensity}
  n63 = Subtract(A=NamedRerouteUsage(), B=LinearInterpolate(Alpha=Multiply(A=@n61, B=@n62) {constB=0.5}) {constA=0.15, constB=0.8})
  n64 = LinearInterpolate(Alpha=@n63)
  n65 = ScalarParameter() {defaultValue=3, group=*S25_SL, parameterName=Secondary_Power}
  n66 = Add(A=Saturate(None=Multiply(A=@n60, B=@n64)), B=Multiply(A=Power(Base=@n63, Exp=@n65) {constExponent=8}) {constB=5})
  n67 = LinearInterpolate(Alpha=NamedRerouteUsage()) {constB=0.5}
  n68 = Multiply(A=Saturate(None=LinearInterpolate(B=@n66, Alpha=@n67)), B=NamedRerouteUsage())
  n69 = ScalarParameter() {defaultValue=5, group=*S25_SL, parameterName=SecondaryPattern_Intensity}
  n70 = NamedRerouteUsage()
  n71 = NamedRerouteUsage()
  n72 = ScalarParameter() {defaultValue=15, group=*S25_SL, parameterName=Secondary_FresnelExp}
  n73 = Saturate(None=Multiply(A=Power(Base=@n71, Exp=@n72), B=ScalarParameter() {defaultValue=70, group=*S25_SL, parameterName=Secondary_FresnelIntesnity}))
  n74 = LinearInterpolate(A=NamedRerouteUsage(), B=@n70, Alpha=@n73)
  n75 = Add(A=NamedRerouteDeclaration(None=Multiply(A=@n68, B=@n69) {constB=15}) {name=SecondaryPattern_Color, nodeColor=<rgba 0.7344 1 0 1>}, B=@n74)
  n76 = Saturate(None=Power(Base=NamedRerouteUsage(), Exp=ScalarParameter() {group=*S25_SL, parameterName=AOPower}))
  NamedRerouteDeclaration(None=Multiply(
    A=StaticSwitchParameter(True=@n75, False=@n74) {group=*S25_SL, parameterName=UseT3}
    B=NamedRerouteDeclaration(None=@n76) {name=AO_Multi, nodeColor=<rgba 0 1 0.7188 1>}
   )) {name=Secondary_BaseColor, nodeColor=<rgba 0 0.5781 1 1>}
  NamedRerouteDeclaration(None=Fresnel() {baseReflectFraction=0, exponent=1}) {name=Fresnel, nodeColor=<rgba 0 1 0.1797 1>}
  NamedRerouteDeclaration(None=@n12.Metallic) {name=Metallic_IN, nodeColor=<rgba 0 0.3906 1 1>}
  n77 = NamedRerouteUsage()
  n78 = NamedRerouteUsage()
  n79 = NamedRerouteUsage()
  n80 = LinearInterpolate(
   A=LinearInterpolate(A=LinearInterpolate(A=@n77, B=@n78, Alpha=@n79), B=NamedRerouteUsage(), Alpha=NamedRerouteUsage())
   B=NamedRerouteUsage()
   Alpha=NamedRerouteUsage()
  )
  n81 = NamedRerouteUsage()
  n82 = LinearInterpolate(
   A=@n77
   B=ScalarParameter() {defaultValue=1, group=*S25_SL, parameterName=NotSL_Metallic}
   Alpha=ScalarParameter() {defaultValue=1, group=*S25_SL, parameterName=NotSL_Metallic_Blend}
  )
  n83 = NamedRerouteUsage()
  n84 = OneMinus(None=NamedRerouteUsage())
  NamedRerouteDeclaration(None=LinearInterpolate(
    A=LinearInterpolate(A=@n80, Alpha=@n81) {constB=0}
    B=NamedRerouteDeclaration(None=@n82) {name=Metallic_Blend, nodeColor=<rgba 1 0.0234 0 1>}
    Alpha=Multiply(A=@n83, B=@n84)
   )) {name=Metallic_OUT, nodeColor=<rgba 0.0364 0.3906 0.379 1>}
  NamedRerouteDeclaration(None=@n12.Roughness) {name=Roughness_IN, nodeColor=<rgba 0.0547 0 1 1>}
  n85 = LinearInterpolate(A=NamedRerouteUsage(), B=NamedRerouteUsage(), Alpha=NamedRerouteUsage())
  n86 = NamedRerouteUsage()
  n87 = NamedRerouteUsage()
  NamedRerouteDeclaration(None=LinearInterpolate(A=LinearInterpolate(A=@n85, B=@n86, Alpha=@n87), B=NamedRerouteUsage(), Alpha=NamedRerouteUsage())) {name=Roughness_OUT, nodeColor=<rgba 0 1 0.2969 1>}
  NamedRerouteDeclaration(None=LinearInterpolate(A=NamedRerouteUsage(), B=NamedRerouteUsage(), Alpha=NamedRerouteUsage())) {name=BaseColor_Hair_OUT, nodeColor=<rgba 0.1277 1 0.2754 1>}
  NamedRerouteDeclaration(None=@n12.Specular) {name=Specular_IN, nodeColor=<rgba 0.875 1 0 1>}
  NamedRerouteDeclaration(None=LinearInterpolate(A=NamedRerouteUsage(), B=NamedRerouteUsage(), Alpha=NamedRerouteUsage())) {name=Specular_OUT, nodeColor=<rgba 0.3125 1 0 1>}
  n88 = NamedRerouteUsage()
  n89 = ScalarParameter() {defaultValue=1, group=*S25_SL, parameterName=Tertiary_FresnelExp}
  n90 = Add(
   A=ScalarParameter() {defaultValue=0.5, group=*S25_SL, parameterName=Tertiary_FresnelBias}
   B=Multiply(A=Power(Base=@n88, Exp=@n89), B=ScalarParameter() {defaultValue=1, group=*S25_SL, parameterName=Tertiary_FresnelIntesnity})
  ) {constA=1}
  NamedRerouteDeclaration(None=LinearInterpolate(A=NamedRerouteUsage(), B=NamedRerouteUsage(), Alpha=Saturate(None=@n90))) {name=Tertiary_BaseColor, nodeColor=<rgba 0.8281 1 0 1>}
  NamedRerouteDeclaration(None=LinearInterpolate(
    A=ScalarParameter() {defaultValue=0.5, group=*S25_SL, parameterName=Primary_Metallic_Max}
    B=ScalarParameter() {group=*S25_SL, parameterName=Primary_Metallic_Min}
    Alpha=@n55
   ) {constA=0.5, constB=0}) {name=Primary_Metallic, nodeColor=<rgba 0.9922 0 1 1>}
  n91 = ScalarParameter() {defaultValue=0.3, group=*S25_SL, parameterName=Primary_Roughness_Min}
  n92 = ScalarParameter() {defaultValue=0.8, group=*S25_SL, parameterName=Primary_Roughness_Max}
  NamedRerouteDeclaration(None=StaticSwitchParameter(
    True=NamedRerouteDeclaration(None=@n58.Roughness) {name=Fur_Roughness, nodeColor=<rgba 0.9453 1 0 1>}
    False=LinearInterpolate(A=@n91, B=@n92, Alpha=@n55) {constA=0.3, constB=0.8}
   ) {parameterName=UseT1}) {name=Primary_Roughness, nodeColor=<rgba 1 0.2451 0.5971 1>}
  NamedRerouteDeclaration(None=LinearInterpolate(
    A=ScalarParameter() {defaultValue=0.2, group=*S25_SL, parameterName=Primary_Specular_Min}
    B=ScalarParameter() {defaultValue=0.5, group=*S25_SL, parameterName=Primary_Specular_Max}
    Alpha=@n55
   ) {constB=0.5}) {name=Primary_Specular, nodeColor=<rgba 1 0.1126 0.0828 1>}
  n93 = StaticSwitchParameter(True=VectorParameter() {group=*S25_SL, parameterName=HairColor, defaultValue=<rgba 0.7656 0.739 0.4865 1>}.RGB, False=NamedRerouteUsage()) {group=*S25_SL, parameterName=UseT2}
  n94 = NamedRerouteUsage()
  n95 = ScalarParameter() {defaultValue=1, group=*S25_SL_Hair, parameterName=Hair_FresnelExp}
  n96 = LinearInterpolate(
   A=@n93
   B=NamedRerouteUsage()
   Alpha=Multiply(A=Power(Base=@n94, Exp=@n95), B=ScalarParameter() {defaultValue=1, group=*S25_SL_Hair, parameterName=Hair_FresnelIntesnity})
  ) {desc=add anture hair color}
  n97 = NamedRerouteUsage()
  n98 = NamedRerouteUsage()
  n99 = ScalarParameter() {defaultValue=0.25, group=*S25_SL_Hair, parameterName=Hair_SecondaryMask_Offset}
  n100 = SmoothStep(Value=Power(Base=Subtract(A=@n98, B=@n99), Exp=ScalarParameter() {defaultValue=2, group=*S25_SL_Hair, parameterName=Hair_SecondaryMask_Power}))
  n101 = LinearInterpolate(
   A=LinearInterpolate(
    A=LinearInterpolate(A=@n93, B=@n96, Alpha=@n97)
    B=NamedRerouteUsage()
    Alpha=NamedRerouteDeclaration(None=@n100) {name=Secondary_Mask_Hair_Adjusted, nodeColor=<rgba 1 0 0.2344 1>}
   )
   B=NamedRerouteUsage()
   Alpha=NamedRerouteUsage()
  )
  n102 = NamedRerouteUsage()
  NamedRerouteDeclaration(None=StaticSwitchParameter(True=LinearInterpolate(A=@n101, Alpha=@n102) {constB=0}, False=@n101) {group=*S25_SL_Hair, parameterName=UseHairEmissive}) {name=SL_BaseColor_Hair, nodeColor=<rgba 0 1 0.6484 1>}
  NamedRerouteDeclaration(None=ScalarParameter() {defaultValue=0.8, group=*S25_SL, parameterName=Secondary_Roughness}) {name=Secondary_Roughness, nodeColor=<rgba 0 0.5781 1 1>}
  NamedRerouteDeclaration(None=ScalarParameter() {group=*S25_SL, parameterName=Secondary_Metallic}) {name=Secondary_Metallic, nodeColor=<rgba 0 0.2319 0.401 1>}
  NamedRerouteDeclaration(None=ScalarParameter() {defaultValue=0.8, group=*S25_SL, parameterName=Tertiary_Roughness}) {name=Tertiary_Roughness, nodeColor=<rgba 0 1 0.2352 1>}
  n103 = TextureObjectParameter() {group=*S25_SL, parameterName=SL_FX_Mask, samplerType=SAMPLERTYPE_Masks, texture=/Game/Global/Textures/Default/Blanks/AllRed}
  n104 = NamedRerouteDeclaration(None=TextureSample(Tex=@n103).RGBA) {name=SL_SkinFX, nodeColor=<rgba 1 0.1406 0 1>}
  n105 = ChannelMaskParameter(None=@n104) {group=*S25_SL, maskChannel=Alpha, parameterName=EmissiveChannel, sortPriority=5, defaultValue=<rgba 0 0 0 1>}
  n106 = NamedRerouteDeclaration(None=Multiply(A=@n105.R, B=ScalarParameter() {defaultValue=1, group=*S25_SL, parameterName=EmissiveChannel_Intensity})) {name=Emissive_Mask, nodeColor=<rgba 0.5729 0.5729 0.5729 1>}
  n107 = ChannelMaskParameter(None=@n104) {group=*S25_SL_Hair, maskChannel=Green, parameterName=Hair_PrimaryChannel, sortPriority=3, defaultValue=<rgba 0 1 0 0>}
  n108 = NamedRerouteDeclaration(None=@n107.R) {name=Primary_Mask_Hair, nodeColor=<rgba 0 1 0.099 1>}
  n109 = NamedRerouteDeclaration(None=ChannelMaskParameter(None=@n104) {group=*S25_SL_Hair, parameterName=Hair_SecondaryChannel, sortPriority=4}.R) {name=Secondary_Mask_Hair, nodeColor=<rgba 1 0.0253 0 1>}
  n110 = ChannelMaskParameter(None=@n104) {group=*S25_SL_Hair, maskChannel=Blue, parameterName=Hair_TertiaryChannel, sortPriority=4, defaultValue=<rgba 0 0 1 0>}
  n111 = Add(A=@n106, B=Add(A=Add(A=@n108, B=@n109), B=NamedRerouteDeclaration(None=@n110.R) {name=Tertiary_Mask_Hair, nodeColor=<rgba 0.0439 0 1 1>}))
  NamedRerouteDeclaration(None=Saturate(None=OneMinus(None=@n111))) {name=NotSL_Hair, nodeColor=<rgba 1 0 0.75 1>}
  n112 = VectorParameter() {group=*S25_SL, parameterName=PrimaryInnerColor, sortPriority=0, defaultValue=<rgba 0.009 0.009 0.009 1>}
  NamedRerouteDeclaration(None=@n112.RGB) {name=PrimaryInnerColor, nodeColor=<rgba 0.9453 0 1 1>}
  n113 = VectorParameter() {group=*S25_SL, parameterName=SecondaryInnerColor, sortPriority=1, defaultValue=<rgba 0.009 0.009 0.009 1>}
  NamedRerouteDeclaration(None=@n113.RGB) {name=SecondaryInnerColor, nodeColor=<rgba 0.9507 0.099 1 1>}
  n114 = VectorParameter() {group=*S25_SL, parameterName=TertiaryInnerColor, sortPriority=2, defaultValue=<rgba 0.009 0.009 0.009 1>}
  NamedRerouteDeclaration(None=@n114.RGB) {name=TertiaryInnerColor, nodeColor=<rgba 0.9721 0.4896 1 1>}
  NamedRerouteDeclaration(None=ScalarParameter() {group=*S25_SL, parameterName=Teritary_Metallic}) {name=Tertiary_Metallic, nodeColor=<rgba 0 0.2319 0.401 1>}
  n115 = NamedRerouteUsage()
  n116 = ScalarParameter() {defaultValue=50, group=*S25_SL, parameterName=Primary_TextureSize_B}
  n117 = Add(A=NamedRerouteUsage(), B=Multiply(A=Time(), B=ScalarParameter() {defaultValue=8, group=*S25_SL, parameterName=Primary_Perlin_Speed}))
  n118 = StaticBool() {value=true}
  n119 = NamedRerouteUsage()
  n120 = Constant()
  NamedRerouteDeclaration(None=MaterialFunctionCall(Float3=MaterialFunctionCall(
     TextureObject=@n115
     TextureSize=@n116
     WorldPosition=@n117
     Export Float 4=@n118
     World Space Normal=@n119
     ProjectionTransitionContrast=@n120
    ) {materialFunction=&WorldAlignedTexture}.XY Texture) {desc=b, materialFunction=&BreakOutFloat3Components}.B) {name=Perlin, nodeColor=<rgba 1 0 0.1641 1>}
  NamedRerouteDeclaration(None=@n1.Pre skinned Local Position) {name=PreskinLocalPos, nodeColor=<rgba 1 0.7734 0 1>}
  n121 = NamedRerouteUsage()
  n122 = Add(
   A=Multiply(A=TextureCoordinate(), B=ScalarParameter() {defaultValue=4, group=*S25_SL, parameterName=SecondaryPattern_TextureSize}) {constB=8}
   B=Multiply(
    A=ScalarParameter() {defaultValue=0.01, group=*S25_SL, parameterName=SecondaryPattern_NormalWarpIntensity}
    B=ComponentMask(None=@n121) {g=true, r=true}
   )
  )
  n123 = NamedRerouteUsage()
  NamedRerouteDeclaration(None=Multiply(
    A=TextureSample(UVs=@n122, Tex=@n123) {samplerSource=SSM_Wrap_WorldGroupSettings, samplerType=SAMPLERTYPE_Masks, texture=/Game/Maps/Textures/checkerGrid}.R
    B=ScalarParameter() {defaultValue=1, group=*S25_SL, parameterName=SecondaryPatternMask_Strength}
   )) {name=SecondaryPattern_Raw, nodeColor=<rgba 1 0 0.9727 1>}
  NamedRerouteDeclaration(None=NamedRerouteUsage()) {name=UV_Raw, nodeColor=<rgba 0.3359 0 1 1>}
  n124 = VectorParameter() {group=*S25_SL, parameterName=PrimaryEmissiveColor, sortPriority=0, defaultValue=<rgba 0.9428 0 1 1>}
  NamedRerouteDeclaration(None=@n124.RGB) {name=PrimaryEmissiveColor, nodeColor=<rgba 0.6673 0 1 1>}
  n125 = NamedRerouteUsage()
  n126 = NamedRerouteUsage()
  n127 = Multiply(A=Power(Base=NamedRerouteUsage()) {constExponent=4}) {constB=5}
  n128 = Multiply(A=NamedRerouteDeclaration(None=Saturate(None=@n127)) {name=EmissiveFresnel, nodeColor=<rgba 0.1953 0 1 1>}, B=NamedRerouteUsage())
  n129 = NamedRerouteUsage()
  n130 = Max(A=Subtract(A=@n128, B=OneMinus(None=@n129))) {constB=0}
  n131 = LinearInterpolate(
   B=Desaturation(
    None=StaticSwitchParameter(True=@n125, False=@n126) {group=*S25_SL, parameterName=UseCustomColorAsEmissive}
    Fraction=ScalarParameter() {defaultValue=-2, group=*S25_SL, parameterName=EmissiveDesat}
   )
   Alpha=Multiply(
    A=StaticSwitchParameter(True=@n130, False=@n128) {group=*S25_SL, parameterName=UseT3}
    B=ScalarParameter() {defaultValue=1, group=*S25_SL, parameterName=PrimaryFresnelEmissive_Intensity}
   )
  )
  n132 = Multiply(
   A=NamedRerouteUsage()
   B=Multiply(A=ScalarParameter() {defaultValue=1, group=*S25_SL, parameterName=SecondaryFresnelEmissive_Intensity}, B=NamedRerouteUsage())
  )
  n133 = NamedRerouteUsage()
  NamedRerouteDeclaration(None=LinearInterpolate(A=LinearInterpolate(A=@n131, B=@n132, Alpha=@n133), B=NamedRerouteUsage(), Alpha=NamedRerouteUsage())) {name=Emissive_OUT, nodeColor=<rgba 1 0 0.5615 1>}
  NamedRerouteDeclaration(None=@n12.Emissive Color) {name=Emissive_IN, nodeColor=<rgba 0.8984 0 1 1>}
  NamedRerouteDeclaration(None=@n12.Base Color) {name=BaseColor_IN, nodeColor=<rgba 0 1 0.0156 1>}
  NamedRerouteDeclaration(None=LinearInterpolate(A=NamedRerouteUsage(), B=NamedRerouteUsage(), Alpha=NamedRerouteUsage())) {name=Metallic_Hair_OUT, nodeColor=<rgba 0.3594 0 1 1>}
  NamedRerouteDeclaration(None=VectorParameter() {group=*S25_SL, parameterName=NotSL_Color, defaultValue=<rgba 0.05 0.05 0.05 1>}.RGB) {name=NotSL_Color, nodeColor=<rgba 0 0.7656 1 1>}
  n134 = LinearInterpolate(B=NamedRerouteUsage(), Alpha=Multiply(A=NamedRerouteUsage(), B=NamedRerouteUsage()))
  n135 = NamedRerouteUsage()
  n136 = NamedRerouteUsage()
  NamedRerouteDeclaration(None=StaticSwitchParameter(True=LinearInterpolate(A=@n134, B=@n135, Alpha=@n136), False=@n134) {group=*S25_SL_Hair, parameterName=UseHairEmissive}) {name=Emissive_Hair_OUT, nodeColor=<rgba 0.0312 0 1 1>}
  n137 = NamedRerouteUsage()
  n138 = ScalarParameter() {defaultValue=80, group=*S25_SL, parameterName=Primary_TextureSize_G}
  n139 = VectorParameter() {group=*S25_SL, parameterName=PrimaryPattern_WarpDirection_01, defaultValue=<rgba 3.156 -1.486 -2.4186 1>}
  n140 = Time()
  n141 = NamedRerouteUsage()
  n142 = Add(A=Multiply(A=@n139.RGB, B=@n140), B=@n141)
  n143 = NamedRerouteUsage()
  n144 = NamedRerouteUsage()
  n145 = VectorParameter() {group=*S25_SL, parameterName=PrimaryPattern_WarpDirection_02, defaultValue=<rgba -3.1565 1.249 2.5496 1>}
  n146 = Add(A=@n141, B=Multiply(A=@n140, B=@n145.RGB))
  n147 = Multiply(
   A=MaterialFunctionCall(Float3=MaterialFunctionCall(TextureObject=@n137, TextureSize=@n138, WorldPosition=@n142, World Space Normal=@n143) {materialFunction=&WorldAlignedTexture}.XY Texture) {materialFunction=&BreakOutFloat3Components}.G
   B=MaterialFunctionCall(Float3=MaterialFunctionCall(TextureObject=@n144, TextureSize=@n138, WorldPosition=@n146, World Space Normal=@n143) {materialFunction=&WorldAlignedTexture}.XY Texture) {materialFunction=&BreakOutFloat3Components}.G
  ) {desc=g}
  NamedRerouteDeclaration(None=Multiply(
    A=NamedRerouteDeclaration(None=@n147) {name=PrimaryPattern_Warp_Raw, nodeColor=<rgba 0.4766 1 0 1>}
    B=ScalarParameter() {defaultValue=0.5, group=*S25_SL, parameterName=PrimaryPattern_WarpIntensity}
   ) {constB=0.5}) {name=PrimaryPattern_Warp, nodeColor=<rgba 0.7812 1 0 1>}
  n148 = TextureObjectParameter() {
   group=*S25_SL
   parameterName=Rainbow_Tex
   samplerType=SAMPLERTYPE_LinearColor
   texture=/BRCosmetics/Characters/Player/Female/Medium/Bodies/F_MED_CinderMax/Skins/SuperLevel/S25_SL_Parent_Materials/Textures/T_S25_SL_TA_T2_Rainbow
  }
  n149 = Constant() {r=130}
  n150 = Add(A=NamedRerouteUsage(), B=Multiply(A=Time(), B=Constant3Vector() {constant=<rgba -10 -10 10 0>}))
  n151 = Constant()
  n152 = NamedRerouteUsage()
  NamedRerouteDeclaration(None=Multiply(
    A=MaterialFunctionCall(TextureObject=@n148, TextureSize=@n149, WorldPosition=@n150, World Space Normal=@n119, ProjectionTransitionContrast=@n151) {materialFunction=&WorldAlignedTexture}.XY Texture
    B=LinearInterpolate(A=@n152)
   )) {name=PerlinRGB, nodeColor=<rgba 1 0.9609 0 1>}
  n153 = MaterialFunctionCall(Float3=@n1.M) {materialFunction=&BreakOutFloat3Components}
  NamedRerouteDeclaration(None=@n153.R) {name=AO, nodeColor=<rgba 0.1016 0 1 1>}
  NamedRerouteDeclaration(None=VectorParameter() {group=*S25_SL, parameterName=PrimaryColor, sortPriority=0, defaultValue=<rgba 1 0 0 1>}.RGB) {name=PrimaryColor, nodeColor=<rgba 0.9453 0 1 1>}
  NamedRerouteDeclaration(None=ScalarParameter() {defaultValue=0.8, group=*S25_SL, parameterName=Secondary_Specular}) {name=Secondary_Specular, nodeColor=<rgba 0 0.5781 1 1>}
  n154 = TextureObjectParameter() {
   group=*S25_SL
   parameterName=SL_TA_Tex
   samplerType=SAMPLERTYPE_LinearColor
   texture=/BRCosmetics/Characters/Player/Female/Medium/Bodies/F_MED_CinderMax/Skins/SuperLevel/S25_SL_Parent_Materials/Textures/T_S25_SL_TA_T1
  }
  NamedRerouteDeclaration(None=@n154) {name=TA_Tex, nodeColor=<rgba 0.8893 0.9099 1 1>}
  n155 = NamedRerouteUsage()
  n156 = NamedRerouteUsage()
  n157 = ComponentMask(None=@n156) {g=true, r=true}
  n158 = VectorParameter() {group=*S25_SL, parameterName=CylindricalProj_Tiling, channelNames {a="", b="", g=V, r=U}, defaultValue=<rgba 3 50 0 1>}
  n159 = AppendVector(
   A=Multiply(
    A=MaterialFunctionCall(Vector or UVs=@n157) {materialFunction=&VectorToRadialValue}.Vector Converted to Angle
    B=@n158.U
   ) {constB=2}
   B=ComponentMask(None=Divide(A=@n156, B=@n158.V)) {b=true}
  )
  n160 = VectorParameter() {
   group=*S25_SL
   parameterName=CylindricalProj_RotationSpeed
   channelNames {a="", b="", g=V_Speed, r=U_Speed}
   defaultValue=<rgba 0.05 0.05 0 1>
  }
  n161 = Multiply(
   A=MaterialFunctionCall(X=@n160.U_Speed, Y=@n160.U_Speed) {materialFunction=&MakeFloat2}.Result
   B=Time()
  )
  n162 = TextureCoordinate()
  n163 = TextureSample!(
   UVs=Add(A=ConstantBiasScale(None=@n155) {bias=-0.5, scale=2}, B=Add(A=@n159, B=@n161))
   Tex=NamedRerouteUsage()
   "DDX(UVs)"=DDX(Value=@n162)
   "DDY(UVs)"=DDY(Value=@n162)
   Apply View MipBias
  ) {
   mipValueMode=TMVM_Derivative
   samplerSource=SSM_Wrap_WorldGroupSettings
   samplerType=SAMPLERTYPE_Masks
   texture=/BRCosmetics/Characters/Player/Female/Medium/Bodies/F_MED_CinderMax/Skins/SuperLevel/S25_SL_Parent_Materials/Textures/T_S25_SL_TA_T3
  }
  n164 = SphereMask(
   A=TextureCoordinate()
   B=ComponentMask(None=VectorParameter() {group=*S25_SL, parameterName=TopMask_Position, defaultValue=<rgba 0 0 0 1>}.RGB) {g=true, r=true}
   Radius=ScalarParameter() {defaultValue=0.2, group=*S25_SL, parameterName=TopMask_Radius}
  ) {attenuationRadius=0.1, hardnessPercent=0}
  n165 = LinearInterpolate(A=@n163.A, Alpha=Saturate(None=Multiply(A=@n164) {constB=1.25})) {constB=0}
  n166 = NamedRerouteUsage()
  n167 = MaterialFunctionCall(
   TextureObject=NamedRerouteUsage()
   TextureSize=ScalarParameter() {defaultValue=50, group=*S25_SL, parameterName=Primary_TextureSize_R}
   WorldPosition=Add(A=NamedRerouteUsage(), B=ConstantBiasScale(None=@n166) {bias=-0.5, scale=2}) {constB=0}
   World Space Normal=NamedRerouteUsage()
   ProjectionTransitionContrast=Constant()
  ) {materialFunction=&WorldAlignedTexture}
  NamedRerouteDeclaration(None=StaticSwitchParameter(
    True=StaticSwitchParameter(True=@n165, False=@n163.A) {group=*S25_SL, parameterName=UseCylinderTopMask}
    False=MaterialFunctionCall(Float3=@n167.XY Texture) {desc=r, materialFunction=&BreakOutFloat3Components}.R
   ) {group=*S25_SL, parameterName=UseT3}) {name=PrimaryPattern_Static, nodeColor=<rgba 0 1 0.5547 1>}
  NamedRerouteDeclaration(None=@n12.Normal) {name=Normal_IN, nodeColor=<rgba 0.6798 0.1602 1 1>}
  NamedRerouteDeclaration(None=VectorParameter() {group=*S25_SL, parameterName=NotSL_Color_Light, defaultValue=<rgba 0.8958 0.8958 0.8958 1>}.RGB) {name=NotSL_Color_Light, nodeColor=<rgba 0 0.7656 1 1>}
  n168 = VectorParameter() {group=*S25_SL, parameterName=PrimaryPatternColor, sortPriority=2, defaultValue=<rgba 0.05 0.05 0.05 1>}
  NamedRerouteDeclaration(None=@n168.RGB) {name=PrimaryPatternColor, nodeColor=<rgba 0.157 0.079 0.1615 1>}
  NamedRerouteDeclaration(None=VertexInterpolator(VS=PreSkinnedNormal()).PS) {name=PreskinLocalNormal, nodeColor=<rgba 1 0.6328 0 1>}
  NamedRerouteDeclaration(None=@n153.G) {name=Gmap, nodeColor=<rgba 0 0.5547 1 1>}
  NamedRerouteDeclaration(None=@n58.Normal) {name=Fur_Normal, nodeColor=<rgba 0.9453 1 0 1>}
  n169 = NamedRerouteUsage()
  n170 = MaterialFunctionCall(BaseNormal=@n169, AdditionalNormal=NamedRerouteUsage()) {materialFunction=&BlendAngleCorrectedNormals}
  n171 = NamedRerouteUsage()
  NamedRerouteDeclaration(None=StaticSwitchParameter(True=LinearInterpolate(A=@n169, B=@n170.Result, Alpha=@n171), False=@n169) {parameterName=UseT1}) {name=Normal_OUT, nodeColor=<rgba 1 0 0.1406 1>}
  n172 = VectorParameter() {group=*S25_SL, parameterName=SecondaryPattern_Color, sortPriority=2, defaultValue=<rgba 0.3177 0.0549 0.3129 1>}
  NamedRerouteDeclaration(None=@n172.RGB) {name=SecondaryPatternColor, nodeColor=<rgba 0.157 0.079 0.1615 1>}
  NamedRerouteDeclaration(None=VectorParameter() {group=*S25_SL, parameterName=EmissiveColor, sortPriority=0, defaultValue=<rgba 1 0 0 1>}.RGB) {name=EmissiveColor, nodeColor=<rgba 0.3348 0 0.3542 1>}
  NamedRerouteDeclaration(None=LinearInterpolate(A=NamedRerouteUsage(), B=@n70, Alpha=@n73)) {name=Secondary_BaseColor_Hair, nodeColor=<rgba 0 0.5781 1 1>}
  n173 = NamedRerouteUsage()
  n174 = ScalarParameter() {defaultValue=3, group=*S25_SL, parameterName=Quaternary_EmissiveIntensity}
  n175 = Power(Base=NamedRerouteUsage()) {constExponent=0.25}
  NamedRerouteDeclaration(None=Multiply(A=Multiply(A=@n173, B=@n174), B=LinearInterpolate(Alpha=@n175) {constA=1, constB=0})) {name=Quaternary_Emissive, nodeColor=<rgba 1 0.3281 0 1>}
  NamedRerouteDeclaration(None=@n153.B) {name=Skin, nodeColor=<rgba 1 0 0.0469 1>}
  n176 = NamedRerouteDeclaration(None=ChannelMaskParameter(None=@n104) {group=*S25_SL, parameterName=PrimaryChannel, sortPriority=3}.R) {name=Primary_Mask, nodeColor=<rgba 1 0.0253 0 1>}
  n177 = ChannelMaskParameter(None=@n104) {group=*S25_SL, maskChannel=Blue, parameterName=SecondaryChannel, sortPriority=4, defaultValue=<rgba 0 0 1 0>}
  n178 = NamedRerouteDeclaration(None=SmoothStep(
    Min=ScalarParameter() {group=*S25_SL, parameterName=SecondaryChannel_Min}
    Max=ScalarParameter() {defaultValue=0.5, group=*S25_SL, parameterName=SecondaryChannel_Max}
    Value=@n177.R
   ) {constMax=0.51, constMin=0.5}) {name=Secondary_Mask, nodeColor=<rgba 0 0 1 1>}
  n179 = ChannelMaskParameter(None=@n104) {group=*S25_SL, maskChannel=Green, parameterName=TertiaryChannel, sortPriority=5, defaultValue=<rgba 0 1 0 0>}
  n180 = Add(A=Add(A=Add(A=@n176, B=@n178), B=NamedRerouteDeclaration(None=@n179.R) {name=Tertiary_Mask, nodeColor=<rgba 0 1 0 1>}), B=@n106)
  NamedRerouteDeclaration(None=OneMinus(None=Saturate(None=@n180))) {name=NotSL, nodeColor=<rgba 0.3125 1 0 1>}
  NamedRerouteUsage()
  Power(Base=NamedRerouteUsage(), Exp=ScalarParameter() {defaultValue=3, group=*S25_SL, parameterName=NotSL_AO_Power}) {constExponent=5}
  ScalarParameter() {defaultValue=1, group=*S25_SL_Hair, parameterName=Hair_Metallic}
  ScalarParameter() {defaultValue=8, group=*S25_SL, parameterName=Primary_RGBPerlin_Speed}
 ]
}
